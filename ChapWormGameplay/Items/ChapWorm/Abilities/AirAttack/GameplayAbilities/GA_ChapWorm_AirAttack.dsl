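; Gameplay ability for the ChapWorm air attack, a stamina-costing pounce dash that blocks melee until a montage notify or ability end.
; Continuous forward targeting starts a tick after the movement extension begins; a hit, timeout or extension end stops the dash.
; A predictive component trace plus downward trace selects the GroundPunch or AirPunch montage section.
; Montage notifies drive deduplicated damage, staggered across ticks, and clamp landing velocity by sprint state.
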(event Ability|EventActivateAbility
  ;[0.04741 0.276 0.08513] ActivateAbility
  (|CallParentFunction)
  ;; Block melee.
  (bind _returnvalue (Ability|ApplyGameplayEffectToOwner "/ChapWormGameplay/Items/ChapWorm/Abilities/AirAttack/GameplayEffects/GE_ChapWorm_AirAttack_BlockMelee.GE_ChapWorm_AirAttack_BlockMelee_C"))
  (Variables|Transient|SetBlockMeleeGameplayEffectHandle _returnvalue)
  (Dash|ConsumeStamina))

(event EventOnMMEStart
  ;[0.06319 0.3281 0.2339 0.5] Handle MME events
  (|CallParentFunction)
  (bind _returnvalue (Utilities|Time|SetTimerforNextTickbyEvent (EventDispatchers|CreateEvent :!event Custom|BeginTargeting))) ; Optimization to prevent targeting traces from running on the same frame that the MME begins.
  (Variables|Transient|SetBeginTargetingFrameDelayTimerHandle _returnvalue)
  (bind _asynctaskproxy (GameplayTasks|TaskWaitDelay 0 (ScalableFloat|GetValueatLevel (Variables|Hotfix|GetHotfix_DashDurationBeforeMontageUpdate)))
    (:then
      (Variables|Transient|AsyncTask|SetAsyncTask_DelayDashMontageUpdate _asynctaskproxy)
      (Ability|RemoveGameplayEffectFromOwnerWithGrantedTags (Variables|Settings|GetSuperSprintThresholdGameplayTag)))
    (:OnFinish
      (CallFunction|EndTargeting)
      (CallFunction|EndDash))))

(event EventOnMMEEnd
  ;[0.06319 0.3281 0.2339 0.5] Handle MME events
  (|CallParentFunction)
  (CallFunction|EndTargeting)
  (CallFunction|EndDash))

(event EventOnMMETimeOut
  ;[0.06319 0.3281 0.2339 0.5] Handle MME events
  (|CallParentFunction)
  (CallFunction|EndTargeting)
  (CallFunction|EndDash))

(event Custom|BeginTargeting
  ;[0.1062 0.01917 0.1094] ForwardTargeting: Look for potential targets in front of the player.
  (bind (_asynctaskproxy _targetdata) (Ability|Tasks|DoTargetSelectionContinuous "ContinuousTargeting" "(TagName=\"Gameplay.Application.Hit\")")
    (:then
      (bind _output_get (Variables|Transient|AsyncTask|SetAsyncTask_ContinuousTargeting _asynctaskproxy)))
    (:Targeted
      (if (Utilities|Array|IsNotEmpty (Ability|TargetData|GetAllActorsfromTargetData _targetdata))
        (GameplayTasks|EndTask _output_get)
        (Variables|Transient|SetContinuousTargetingHitResult (Ability|TargetData|GetHitResultfromTargetData _targetdata))
        (CallFunction|EndDash)))))

(event Custom|EndTargeting
  ;[0.1062 0.01917 0.1094] ForwardTargeting: Look for potential targets in front of the player.
  (GameplayTasks|EndTask (Variables|Transient|AsyncTask|GetAsyncTask_ContinuousTargeting))
  (Utilities|Time|ClearandInvalidateTimerbyHandle (Variables|Transient|GetBeginTargetingFrameDelayTimerHandle)))

(event EventEndDash
  ;[0.03101 0.1119 0.1458] EndDash: The dash movement has ended either due to finding an target along the path or due to the dash MME ending.
  ; Jump to the next montage section.
  (|CallParentFunction)
  (bind _returnvalue (Dash|TraceClosetoGround))
  (Ability|Animation|MontageJumptoSection (select _returnvalue "GroundPunch" "AirPunch")))

(event EventCastMontageNotifyStart
  (params (NotifyTag :type Gameplay%20Tag%20Structure))
  ;[0.01664 0.1007 0.1094] NotifyStart: When the impact notification is received, perform the damage execution.
  (|CallParentFunction NotifyTag)
  (if (== NotifyTag "(TagName=\"ChapWorm.AnimNotify.Pounce.Dash.Impact\")")
    (bind _outfortplayerpawn (CallFunction|GetAvatarasValidFortPlayerPawn)
      (:ValidFortPlayerPawn
        ;; Performs the target selection that eventually applies the damage of the attack.
        (bind (_asynctaskproxy _targetdata _applicationtag) (Ability|Tasks|DoTargetSelection "DamageTargetSelection" (select (GameplayTags|HasMatchingGameplayTag _outfortplayerpawn "(TagName=\"ChapWorm.Overpower\")") "(TagName=\"ChapWorm.AnimNotify.Pounce.Dash.Impact.Overpower\")" "(TagName=\"ChapWorm.AnimNotify.Pounce.Dash.Impact.Default\")"))
          (:then
            (Variables|Transient|AsyncTask|SetAsyncTask_DamageTargetSelection _asynctaskproxy)
            (if (not (Variables|Transient|GetbUsedMoveToActorMME))
              (bind _outfortplayerpawn_2 (CallFunction|GetAvatarasValidFortPlayerPawn)
                (:ValidFortPlayerPawn
                  (bind _movementmoderuntimedata (Variables|Transient|GetMovementModeRuntimeData)
                    (:then
                      (Class|FortMovementModeExtBlueprintLibrary|EndMovementModeExtensiononActor _movementmoderuntimedata _outfortplayerpawn_2))
                    (:else))))))
          (:Targeted
            (bind _filteredtargetdata (Ability|FilterTargetData _targetdata))
            (Variables|Transient|SetCachedDamageTargetDataHandle _filteredtargetdata)
            (Variables|Transient|SetCachedDamageApplicationTag _applicationtag)
            (Utilities|FlowControl|DelayUntilNextTick) ; Perf
            (Ability|ApplyGameplayEffectContainer (Variables|Transient|GetCachedDamageTargetDataHandle) (Variables|Transient|GetCachedDamageApplicationTag))
            (Utilities|FlowControl|DelayUntilNextTick) ; Perf
            (Ability|ExecuteGameplayCueOnOwner (select (Ability|TargetData|TargetDataHasHitResult (Variables|Transient|GetCachedDamageTargetDataHandle)) (Variables|Settings|GetImpactTargetSelectionSuccessGameplayCueTag) (Variables|Settings|GetImpactTargetSelectionFailGameplayCueTag)))))))
    (elif (== NotifyTag "(TagName=\"ChapWorm.AnimNotify.Pounce.Dash.EndOnGround\")")
      ;;[0.1615 0.02428 0.1374] Dash has touched the ground, figure out at what velocity the player should exit at based on player's movement state.
      (bind _outfortplayerpawn_3 (CallFunction|GetAvatarasValidFortPlayerPawn)
        (:ValidFortPlayerPawn
          (bind _charactermovement (Variables|/Script/Engine.Character|GetCharacterMovement _outfortplayerpawn_3)
            (:then
              (Variables|/Script/Engine.MovementComponent|SetVelocity (Math|Vector|ClampVectorSize (Variables|/Script/Engine.MovementComponent|GetVelocity _charactermovement) 0.0 (ScalableFloat|GetValueatLevel (select (Variables|/Script/FortniteGame.FortPlayerPawn|GetbIsTacticalSprinting _outfortplayerpawn_3) (Variables|Hotfix|GetHotfix_DashOnGroundTacSprintVelocityMaxClamp) (select (PawnStats|IsSprinting _outfortplayerpawn_3) (Variables|Hotfix|GetHotfix_DashOnGroundRunningVelocityMaxClamp) (Variables|Hotfix|GetHotfix_DashOnGroundWalkVelocityMaxClamp))))) _charactermovement))
            (:else))))
      (elif (== NotifyTag "(TagName=\"ChapWorm.AnimNotify.Pounce.AllowMelee.Server\")")
        ;;[0.15 0.02207 0.1374 0.5] At a certain point in the air attack, the player is allowed to execute their mantis attacks again.
        (Ability|HasAuthority
          (:True
            (Ability|RemoveGameplayEffectFromOwnerWithHandle (Variables|Transient|GetBlockMeleeGameplayEffectHandle))))
        (elif (== NotifyTag "(TagName=\"ChapWorm.AnimNotify.Pounce.AllowMelee.Client\")")
          (Ability|HasAuthority
            (:False)))))))

(event EventPlayerEnterWater
  (params (FortPlayerPawn :type Fort%20Player%20Pawn%20Object%20Reference))
  ;[0.01368 0.1323 0.1615] PlayerEnterWater: Apply the EnterWater gameplay effect, which will be responsible for applying the gameplay cue.
  (|CallParentFunction FortPlayerPawn)
  (Ability|ApplyGameplayEffectToOwner "/ChapWormGameplay/Items/ChapWorm/Abilities/AirAttack/GameplayEffects/GE_ChapWorm_AirAttack_EnterWater.GE_ChapWorm_AirAttack_EnterWater_C"))

(event Ability|EventOnEndAbility
  (params (bWasCancelled :type Boolean))
  ;[0.08333 0.0335 0.007236] OnEndAbility: Clean up.
  (Utilities|FlowControl|Sequence
    (:then_0
      (GameplayTasks|EndTask (Variables|Transient|AsyncTask|GetAsyncTask_ContinuousTargeting))
      (Ability|RemoveGameplayEffectFromOwnerWithHandle (Variables|Transient|GetBlockMeleeGameplayEffectHandle))
      (CallFunction|EndTargeting))
    (:then_1
      (|CallParentFunction bWasCancelled))))

(fn ConsumeStamina
  ; Consume the stamina cost for using this ability.
  (if (not (Variables|Transient|GetbConsumedStamina))
    (bind _tacticalsprintenergycomponent (Default|GetTacticalSprintEnergyComponent)
      (:IsValid
        (Energy|UseEnergy _tacticalsprintenergycomponent (ScalableFloat|GetValueatLevel (Variables|Hotfix|GetHotfix_EnergyCost)) true self)
        (Variables|Transient|SetbConsumedStamina true)))))

(fn ExecuteFailGameplayCue
  (params (Target :type Actor%20Object%20Reference))
  (GameplayCue|ExecuteGameplayCue%28Burst%29Local Target (Variables|Settings|GetGameplayCueTagActivateFailed) (Ability|GameplayCue|MakeGameplayCueParameters 0.0)))

(fn FilterTargetData
  (params (TargetData :type Gameplay%20Ability%20Target%20Data%20Handle%20Structure))
  ; FilterTargetData: Removing duplicate actor entries.
  (for _index (range (Ability|TargetData|GetDataCountfromTargetData TargetData))
    (Utilities|IsValid (.HitActor (Collision|BreakHitResult (Ability|TargetData|GetHitResultfromTargetData TargetData _index)))
      (:Is%20Valid
        (bind _returnvalue (Utilities|Array|AddUnique (LocalVariables|GetHitActors) (.HitActor (Collision|BreakHitResult (Ability|TargetData|GetHitResultfromTargetData TargetData _index)))))
        (if (>= _returnvalue 0)
          (bind _returnvalue_2 (Ability|TargetData|AppendTargetDataHandle (LocalVariables|GetReturnHandle) (Ability|TargetData|AbilityTargetDatafromHitResult (Ability|TargetData|GetHitResultfromTargetData TargetData _index))))
          (LocalVariables|SetReturnHandle _returnvalue_2)))))
  (return (LocalVariables|GetReturnHandle)))

(fn GetCustomAbilitySourceTransform
  (bind _returnvalue (CallFunction|GetAvatarasFortPlayerPawn))
  (bind (_returnvalue_location _returnvalue_rotation _returnvalue_scale) (Math|Transform|BreakTransform (Transformation|GetActorTransform _returnvalue)))
  ;[0.1044 0.01678 0.1146] GetCustomAbilitySourceTransform: This function is used to return the source transform for the gameplay effect containerof 'Gameplay.Application.hit', which is used during continuous targeting while the dash takes place.
  ; IF the player is moving, this returns the velocity direction as the rotation.
  (Utilities|IsValid _returnvalue
    (:Is%20Valid
      (return _returnvalue_location (Math|Rotator|SelectRotator (Math|Rotator|MakeRotfromX (Transformation|GetVelocity _returnvalue)) (Transformation|GetActorRotation _returnvalue) (> (Math|Vector|VectorLength (Transformation|GetVelocity _returnvalue)) 0)) _returnvalue_scale))
    (:Is%20Not%20Valid
      ;;[1 0 0] Warning: Avatar player pawn is NULL.
      (Fortnite|FortShippingLog self "GetCustomAbilitySourceTransform: Avatar player pawn is NULL." true)
      (return "0, 0, 0" "0, 0, 0" "1.000000,1.000000,1.000000"))))

(fn GetTargetingHit
  (bind _return_value (Utilities|GetTargetingComponent)
    (:Is%20Valid
      (return (Class|BChapWormTargetingComponent|GetLatestTargetingTraceHitResult _return_value)))
    (:Is%20Not%20Valid
      (return (|CallParentFunction)))))

(fn CanActivateAbility
  (params (ActorInfo :type Gameplay%20Ability%20Actor%20Info%20Structure) (Handle :type Gameplay%20Ability%20Spec%20Handle%20Structure))
  (bind (_owneractor _avataractor) (Utilities|Struct|BreakGameplayAbilityActorInfo ActorInfo))
  ;[0.0141 0.1341 0.1354] CanActivateAbility: The ability can only be activated if the player is sprinting.
  (bind (_relevanttags _returnvalue) (|CallParentFunction ActorInfo Handle))
  (LocalVariables|SetReturnRelevantTags _relevanttags)
  (Utilities|FlowControl|Sequence
    (:then_0
      (if _returnvalue
        (LocalVariables|SetReturnRelevantTags)
        (bind _asfort_player_pawn (Utilities|Casting|CastToFortPlayerPawn _avataractor))
        (if (not (GameplayTags|HasAnyMatchingGameplayTags _asfort_player_pawn (GameplayTags|MakeLiteralGameplayTagContainer "(GameplayTags=((TagName=\"Gameplay.Action.Player.SprintJumping\"),(TagName=\"ChapWorm.Ability.SuperSprint.Threshold\")))")))
          (if (GameplayTags|HasAnyMatchingGameplayTags _asfort_player_pawn (GameplayTags|MakeLiteralGameplayTagContainer "(GameplayTags=((TagName=\"MovementMode.Ballooning\"),(TagName=\"MovementMode.Falling\"),(TagName=\"MovementMode.Hovering\")))"))
            (Utilities|IsValid (Actor|GetComponentbyClass _asfort_player_pawn "/TacticalSprintGame/Gameplay/TacticalSprintEnergyComponent.TacticalSprintEnergyComponent_C")
              (:Is%20Valid
                (if (not (and (Energy|HasSufficientEnergy (Actor|GetComponentbyClass _asfort_player_pawn "/TacticalSprintGame/Gameplay/TacticalSprintEnergyComponent.TacticalSprintEnergyComponent_C") (ScalableFloat|GetValueatLevel (select (Utilities|Enum|Equal(Enum) (Networking|GetLocalRole _avataractor) "ROLE_Authority") (Variables|Hotfix|GetHotfix_MinimumEnergyCheckServer) (Variables|Hotfix|GetHotfix_MinimumEnergyCheckClient)))) (not (Math|Float|NearlyEqual(Float) (Energy|GetCurrentEnergyPercentage (Actor|GetComponentbyClass _asfort_player_pawn "/TacticalSprintGame/Gameplay/TacticalSprintEnergyComponent.TacticalSprintEnergyComponent_C"))))))
                  ;; Play error sound.
                  (Ability|ExecuteFailGameplayCue _avataractor)
                  (GameplayTags|AddGameplayTag (LocalVariables|GetReturnRelevantTags) "(TagName=\"Athena.Ability.Failed.ChapWorm.Pounce.Activate.NotEnoughStaminaEnergy\")" :!class /Script/GameplayTags.BlueprintGameplayTagLibrary))))))
        (return _relevanttags true)))
    (:then_1
      (return (LocalVariables|GetReturnRelevantTags) false))))

(fn OverrideFailedReason
  (params (FailedReason :type Gameplay%20Tag%20Container%20Structure%20%28by%20ref%29))
  ;[0.04701 0.15 0.1562] OverrideFailReason: Only allow the reason of low stamina cost to show on the HUD.
  (if (GameplayTags|HasTag FailedReason "(TagName=\"Athena.Ability.Failed.ChapWorm.Pounce.Activate.NotEnoughStaminaEnergy\")" true)
    (return "(GameplayTags=((TagName=\"Athena.Ability.Failed.ChapWorm.Pounce.Activate.NotEnoughStaminaEnergy\")))")
    (else
      (return "(GameplayTags=)"))))

(fn RemoveCameraMode
  ;[0.009367 0.09587 0.1042] RemoveCameraMode: Remove the applied camera mode.
  (Utilities|FlowControl|Sequence
    (:then_0
      (bind _outfortplayerpawn (CallFunction|GetAvatarasValidFortPlayerPawn)
        (:ValidFortPlayerPawn
          (Utilities|IsValid (Fortnite|GetFortPlayerControllerfromActor _outfortplayerpawn)
            (:Is%20Valid
              (bind _playercameramanager (Variables|/Script/Engine.PlayerController|GetPlayerCameraManager (Fortnite|GetFortPlayerControllerfromActor _outfortplayerpawn))
                (:then
                  (bind _asfort_player_camera_base (Utilities|Casting|CastToFortPlayerCameraBase _playercameramanager))
                  (Camera|RemoveCameraModeOverride _asfort_player_camera_base (Variables|Camera|GetCameraModeTag) 0 true))
                (:else)))))))
    (:then_1
      (bind _cameramodifier (Variables|Camera|GetCameraModifier)
        (:then
          (bind _asfort_camera_modifier_chap_worm (Utilities|Casting|CastToFortCameraModifier_ChapWorm _cameramodifier))
          (CameraModifier|WindDownandRemove _asfort_camera_modifier_chap_worm :!class /Script/ChapWormGameplayRuntime.FortCameraModifier_ChapWorm))
        (:else)))))

(fn ResetTransientState
  ;[0.08847 0.15 0.03889 0.5] Crucial functionality to keep up to date. Resets all the transient variables to their default values so that the GA can be executed correctly when invoked multiple times per instance.
  (|CallParentFunction)
  (Variables|Transient|SetbConsumedStamina false)
  (Variables|Transient|SetContinuousTargetingHitResult)
  (Variables|Transient|SetBlockMeleeGameplayEffectHandle)
  (Variables|Transient|SetCachedDamageApplicationTag)
  (Variables|Transient|SetCachedDamageTargetDataHandle))

(fn ShouldPlayFailedMontage
  (params (FailedReason :type Gameplay%20Tag%20Container%20Structure%20%28by%20ref%29))
  (return (GameplayTags|HasTag FailedReason "(TagName=\"Activate.Failed.Cost.Stamina\")" true)))

(fn TraceCloseToGround
  (bind _returnvalue (CallFunction|GetAvatarasFortPlayerPawn))
  (bind _returnvalue_2 (Transformation|GetActorLocation _returnvalue))
  (bind _returnvalue_3 (Math|Vector|GetUnitDirection(Vector) _returnvalue_2 (.ImpactPoint (Collision|BreakHitResult (Variables|Transient|GetContinuousTargetingHitResult)))))
  (bind _continuoustargetinghitresult (Variables|Transient|GetContinuousTargetingHitResult))
  (bind (_bblockinghit _binitialoverlap _time _distance _location _impactpoint _normal _impactnormal _physmat _hitactor _hitcomponent) (Collision|BreakHitResult _continuoustargetinghitresult))
  (bind _returnvalue_4 (CallFunction|GetAvatarasFortPlayerPawn))
  ;[0.05596 0.1641 0.2083] Returns if the player is close to the ground.
  (bind _outfortplayerpawn (CallFunction|GetAvatarasValidFortPlayerPawn)
    (:ValidFortPlayerPawn
      (Variables|/Script/Engine.Character|GetCapsuleComponent _outfortplayerpawn
        (:then
          (Utilities|FlowControl|Sequence
            (:then_0
              (Utilities|IsValid _hitcomponent
                (:Is%20Valid
                  ;;[0.15 0.03196 0.1314 0.5] Runs a line component trace (very cheap trace) to predict where the dash is going to end up taking the player so that the ground trace can be run from that found location.
                  (bind (_hitlocation _hitnormal _bonename _outhit _returnvalue_5) (Collision|LineTraceComponent _hitcomponent _returnvalue_2 (+ _returnvalue_2 (* _returnvalue_3 (LocalVariables|GetLineTraceToHitCompDistance))) true false false))
                  (if _returnvalue_5
                    (bind _output_get (LocalVariables|SetTraceStartLocation (+ _hitlocation (* (* _returnvalue_3 -1.0) (+ (Components|Capsule|GetScaledCapsuleRadius (Variables|/Script/Engine.Character|GetCapsuleComponent _returnvalue)) 1e+01)))))
                    (LocalVariables|SetTraceEndLocation (Utilities|Operators|Subtract _output_get (Math|Vector|MakeVector 0.0 0.0 (ScalableFloat|GetValueatLevel (Variables|Hotfix|GetHotfix_GroundDistanceCheck)))))
                    (LocalVariables|SetbPerformTrace true)))
                (:Is%20Not%20Valid
                  (bind _output_get_2 (LocalVariables|SetTraceStartLocation (Transformation|GetActorLocation _returnvalue_4)))
                  (LocalVariables|SetTraceEndLocation (Utilities|Operators|Subtract _output_get_2 (Math|Vector|MakeVector 0.0 0.0 (+ (Components|Capsule|GetScaledCapsuleHalfHeight (Variables|/Script/Engine.Character|GetCapsuleComponent _returnvalue_4)) (ScalableFloat|GetValueatLevel (Variables|Hotfix|GetHotfix_GroundDistanceCheck))))))
                  (LocalVariables|SetbPerformTrace true))))
            (:then_1
              (if (LocalVariables|GetbPerformTrace)
                (bind (_outhit_2 _returnvalue_6) (Collision|LineTraceByChannel (LocalVariables|GetTraceStartLocation) (LocalVariables|GetTraceEndLocation) "TraceTypeQuery10" true (Utilities|Array|MakeArray (CallFunction|GetAvatarasFortPlayerPawn)) "None" true "(R=1.000000,G=0.000000,B=0.000000,A=1.000000)" "(R=0.000000,G=1.000000,B=0.000000,A=1.000000)" 1e+01))
                (return _returnvalue_6)))))
        (:else)))
    (:AvatarCastFailed
      (return false))))
