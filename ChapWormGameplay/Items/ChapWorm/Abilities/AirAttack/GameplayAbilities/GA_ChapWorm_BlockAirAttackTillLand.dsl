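; Gameplay ability for the ChapWorm item that, by its name, blocks the air attack while active, until the player lands.
; If the pawn is not already walking or swimming on activation, it binds landing and movement-mode delegates and waits on tags like vehicle, riding, sliding, grinding and hotfix-gated jetpack.
; Surface-swim detection starts after a delay, and sliding covers skipped landing callbacks.
; EndBlock ends the ability once via timer, and cleanup releases tasks and bindings.
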(event Ability|EventActivateAbility
  ;[0.05824 0.1406 0.0446] ActivateAbility: Set up binds to wait for the player to land.
  (|CallParentFunction)
  (bind _asfort_pawn (Utilities|Casting|CastToFortPawn (Ability|GetAvatarActorfromActorInfo))
    (:then
      (bind _charactermovement (Variables|/Script/Engine.Character|GetCharacterMovement _asfort_pawn)
        (:then
          (if (Math|Boolean|ANDBoolean (Utilities|Enum|NotEqual(Enum) (Variables|/Script/Engine.CharacterMovementComponent|GetMovementMode _charactermovement) "MOVE_Walking") (Utilities|Enum|NotEqual(Enum) (Variables|/Script/Engine.CharacterMovementComponent|GetMovementMode _charactermovement) "MOVE_NavWalking") (Utilities|Enum|NotEqual(Enum) (Variables|/Script/Engine.CharacterMovementComponent|GetMovementMode _charactermovement) "MOVE_Swimming"))
            (CallFunction|SetUpBinds)
            (CallFunction|SetUpWaits)
            (else
              _)))
        (:else
          _))
      (CallFunction|EndBlock))
    (:CastFailed
      _)))

(event Custom|SetUpBinds
  ;[0.1797 0.1039 0.2969] Set Up Binds
  (bind _asfort_player_pawn (Utilities|Casting|CastToFortPlayerPawn (Ability|GetAvatarActorfromActorInfo)))
  (EventDispatchers|BindEventtoOnPawnLanded _asfort_player_pawn (delegate Custom|HandlePawnLanded))
  (Character|BindEventtoMovementModeChangedDelegate _asfort_player_pawn (EventDispatchers|CreateEvent :!event HandleMovementModeChanged)))

(event Custom|HandlePawnLanded
  (params (Hit :type Hit%20Result%20Structure%20%28by%20ref%29))
  ;[0.1797 0.1039 0.2969] Set Up Binds
  (CallFunction|EndBlock))

(event Custom|SetUpWaits
  ;[0.03942 0.1615 0.147] Set Up Waits
  (bind _asynctaskproxy (Ability|Tasks|WaitGameplayTagAdd (Variables|Default|GetTag_InVehicle))
    (:then
      (Utilities|Array|Add (Variables|Transient|GetWaitTagAsyncTasks) _asynctaskproxy)
      (Utilities|FlowControl|Delay (Variables|Default|GetDelay_WaterSwimBind)) ; Don't do the water bind immidately, just incase the player doesn't leave the water in time.
      (bind _asynctaskproxy_2 (Ability|Tasks|WaitGameplayTagAdd (Variables|Default|GetTag_SurfaceSwimming))
        (:then
          (Utilities|Array|Add (Variables|Transient|GetWaitTagAsyncTasks) _asynctaskproxy_2)
          (bind _asynctaskproxy_3 (Ability|Tasks|WaitGameplayTagAdd (Variables|Default|GetTagIsRiding))
            (:then
              (Utilities|Array|Add (Variables|Transient|GetWaitTagAsyncTasks) _asynctaskproxy_3)
              (bind _asynctaskproxy_4 (Ability|Tasks|WaitGameplayTagAdd (Variables|Default|GetTag_Sliding))
                (:then
                  (Utilities|Array|Add (Variables|Transient|GetWaitTagAsyncTasks) _asynctaskproxy_4)
                  (bind _asynctaskproxy_5 (Ability|Tasks|WaitGameplayTagAdd "(TagName=\"MovementMode.Grinding\")")
                    (:then
                      (Utilities|Array|Add (Variables|Transient|GetWaitTagAsyncTasks) _asynctaskproxy_5)
                      (if (ScalableFloat|GetValueasBool (Variables|Default|GetHotfix_ResetWithJetpack))
                        (bind _asynctaskproxy_6 (Ability|Tasks|WaitGameplayTagAdd "(TagName=\"Gameplay.Action.Player.Jetpack\")")
                          (:then
                            (Utilities|Array|Add (Variables|Transient|GetWaitTagAsyncTasks) _asynctaskproxy_6))
                          (:Added
                            _))))
                    (:Added
                      _))
                  (CallFunction|EndBlock))
                (:Added
                  _)) ; When sliding down a hill or cliff that isn't walkable, the OnPawnLanded callback can get skipped.\nEnsure that if the player gets into a sliding state, they are treated as if they had landed.
            )
            (:Added
              _)) ; If the player start to ride creature we have to stop the Immunity too
        )
        (:Added
          _)))
    (:Added
      _)))

(event Custom|EndBlock
  ;[0.09307 0.01635 0.09375] EndBlock: End the ability after a short duration.
  (Utilities|FlowControl|DoOnce false
    (:Completed
      (Utilities|Time|SetTimerbyEvent (EventDispatchers|CreateEvent :!event K2_EndAbility) (Variables|Default|GetEndBlockDelay)))))

(event Ability|EventOnEndAbility
  (params (bWasCancelled :type Boolean))
  ;[0.1667 0.08069 0.01371] EndAbility: Clean up.
  (Utilities|FlowControl|Sequence
    (:then_0
      (GameplayTasks|EndTask (Variables|Transient|GetWaitTagAsyncTasks))
      (bind _asfort_player_pawn (Utilities|Casting|CastToFortPlayerPawn (Ability|GetAvatarActorfromActorInfo)))
      (EventDispatchers|UnbindEventfromOnPawnLanded _asfort_player_pawn (EventDispatchers|CreateEvent :!event Custom|HandlePawnLanded))
      (Character|UnbindEventfromMovementModeChangedDelegate _asfort_player_pawn (EventDispatchers|CreateEvent :!event HandleMovementModeChanged)))
    (:then_1
      (|CallParentFunction bWasCancelled))))

(fn HandleMovementModeChanged
  (params (Character :type Character%20Object%20Reference) (PrevMovementMode :type EMovementMode%20Enum :default "MOVE_None") (PreviousCustomMode :type Byte))
  ;[0.02622 0.1467 0.1667] HandleMovementModeChanged: End block if movement mode enters walking.
  (bind _asfort_player_pawn (Utilities|Casting|CastToFortPlayerPawn Character))
  (bind _charactermovement (Variables|/Script/Engine.Character|GetCharacterMovement _asfort_player_pawn)
    (:then
      (if (Math|Boolean|ORBoolean (Utilities|Enum|Equal(Enum) (Variables|/Script/Engine.CharacterMovementComponent|GetMovementMode _charactermovement) "MOVE_Walking") (Utilities|Enum|Equal(Enum) (Variables|/Script/Engine.CharacterMovementComponent|GetMovementMode _charactermovement) "MOVE_NavWalking") (Utilities|Enum|Equal(Enum) (Variables|/Script/Engine.CharacterMovementComponent|GetMovementMode _charactermovement) "MOVE_Swimming"))
        (CallFunction|EndBlock)
        (Character|UnbindEventfromMovementModeChangedDelegate _asfort_player_pawn (EventDispatchers|CreateEvent :!event HandleMovementModeChanged))))
    (:else)))
